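; Geometry Script DynamicMeshActor art tool for switching material IDs on a Nanite mesh and baking to a target static mesh.
; Rebuild and Update split NaniteMesh by material ID, remap via NewMaterialIDs into a compute mesh, and copy FallbackMesh separately.
; AddNaniteMeshToTargetandSwitchIDs calls Update, whose Bake writes only the fallback copy to TargetMesh, with Nanite enabled.
; The construction script previews TargetMesh twice, once with Nanite disallowed.
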(event EventBeginPlay)

(event Collision|EventActorBeginOverlap
  (params (OtherActor :type Actor%20Object%20Reference)))

(event EventTick
  (params (DeltaSeconds :type Float%20%28single-precision%29)))

(event Events|EventOnRebuildGeneratedMesh
  (params (TargetMesh :type Dynamic%20Mesh%20Object%20Reference))
  (Utilities|FlowControl|Sequence
    (:then_0
      (bind _returnvalue (DynamicMeshComponent|GetDynamicMesh (Variables|DynamicMeshActor|GetDynamicMeshComponent)))
      (bind _output_get (Variables|Default|SetDynamicMesh _returnvalue))
      (GeometryScript|StaticMesh|CopyMeshFromStaticMeshwithSectionMaterials (Variables|1SwitchMaterialIDs|GetNaniteMesh) _output_get (Utilities|Struct|MakeGeometryScriptCopyMeshfromAssetOptions true) (Utilities|Struct|MakeGeometryScriptMeshReadLOD "MaxAvailable" -1)
        (:Success
          (bind _returnvalue_2 (DynamicMeshActor|AllocateComputeMesh))
          (bind _returnvalue_3 (GeometryScript|Utility|CreateDynamicMeshPool))
          (bind (_componentmeshes _componentmaterialids) (GeometryScript|Decomposition|SplitMeshbyMaterialIDs (Variables|Default|GetDynamicMesh) _returnvalue_3))
          (for (_ _array_index) _componentmaterialids
            (bind _returnvalue_4 (GeometryScript|Materials|RemapMaterialIDs (Utilities|Array|Get(acopy) _componentmeshes _array_index) _array_index (Math|Integer|SelectInt (Utilities|Array|Get(acopy) (Variables|1SwitchMaterialIDs|GetNewMaterialIDs) _array_index) _array_index (Utilities|Array|IsValidIndex (Variables|1SwitchMaterialIDs|GetNewMaterialIDs) _array_index))))
            (GeometryScript|MeshEdits|AppendMesh _returnvalue_2 _returnvalue_4 (Math|Transform|MakeTransform "0, 0, 0") false "(CombineMode=EnableAllMatching)"))
          (Variables|Default|SetDynamicMesh _returnvalue_2))))
    (:then_1
      (bind _returnvalue_5 (DynamicMeshActor|AllocateComputeMesh))
      (Variables|Default|SetOriginalFallback _returnvalue_5)
      (GeometryScript|StaticMesh|CopyMeshFromStaticMeshwithSectionMaterials :FromStaticMeshAsset (Variables|1SwitchMaterialIDs|GetFallbackMesh) :ToDynamicMesh (Variables|Default|GetOriginalFallback) :RequestedLOD (Utilities|Struct|MakeGeometryScriptMeshReadLOD "MaxAvailable")))))

(event Custom|Update
  (Utilities|FlowControl|Sequence
    (:then_0
      (Utilities|FlowControl|Sequence
        (:then_0
          (bind _returnvalue (DynamicMeshComponent|GetDynamicMesh (Variables|DynamicMeshActor|GetDynamicMeshComponent)))
          (bind _output_get (Variables|Default|SetDynamicMesh _returnvalue))
          (GeometryScript|StaticMesh|CopyMeshFromStaticMeshwithSectionMaterials (Variables|1SwitchMaterialIDs|GetNaniteMesh) _output_get (Utilities|Struct|MakeGeometryScriptCopyMeshfromAssetOptions true) (Utilities|Struct|MakeGeometryScriptMeshReadLOD "MaxAvailable" -1)
            (:Success
              (bind _returnvalue_2 (DynamicMeshActor|AllocateComputeMesh))
              (bind _returnvalue_3 (GeometryScript|Utility|CreateDynamicMeshPool))
              (bind (_componentmeshes _componentmaterialids) (GeometryScript|Decomposition|SplitMeshbyMaterialIDs (Variables|Default|GetDynamicMesh) _returnvalue_3))
              (for (_ _array_index) _componentmaterialids
                (bind _returnvalue_4 (GeometryScript|Materials|RemapMaterialIDs (Utilities|Array|Get(acopy) _componentmeshes _array_index) _array_index (Math|Integer|SelectInt (Utilities|Array|Get(acopy) (Variables|1SwitchMaterialIDs|GetNewMaterialIDs) _array_index) _array_index (Utilities|Array|IsValidIndex (Variables|1SwitchMaterialIDs|GetNewMaterialIDs) _array_index))))
                (GeometryScript|MeshEdits|AppendMesh _returnvalue_2 _returnvalue_4 (Math|Transform|MakeTransform "0, 0, 0") false "(CombineMode=EnableAllMatching)"))
              (Variables|Default|SetDynamicMesh _returnvalue_2))))
        (:then_1
          (bind _returnvalue_5 (DynamicMeshActor|AllocateComputeMesh))
          (Variables|Default|SetOriginalFallback _returnvalue_5)
          (GeometryScript|StaticMesh|CopyMeshFromStaticMeshwithSectionMaterials :FromStaticMeshAsset (Variables|1SwitchMaterialIDs|GetFallbackMesh) :ToDynamicMesh (Variables|Default|GetOriginalFallback) :RequestedLOD (Utilities|Struct|MakeGeometryScriptMeshReadLOD "MaxAvailable")))))
    (:then_1
      (1SwitchMaterialIDs|Bake))))

(fn AddNaniteMeshToTargetandSwitchIDs
  (CallFunction|Update))

(fn Bake
  (Utilities|FlowControl|Sequence
    (:then_0
      (GeometryScript|StaticMesh|CopyMeshtoStaticMesh (Variables|Default|GetOriginalFallback) (Variables|1SwitchMaterialIDs|GetTargetMesh) (Utilities|Struct|MakeGeometryScriptCopyMeshtoAssetOptions :bEnableRecomputeNormals false :bEnableRecomputeTangents false :bEnableRemoveDegenerates false :BoneHierarchyMismatchHandling "DoNothing" :bRemapBoneIndicesToMatchAsset false :bUseOriginalVertexOrder false :bUseBuildScale true :bReplaceMaterials false :NewMaterialUsage "ReplaceAssetMaterials" :bCleanAssignedMaterials false :GenerateLightmapUVs "MatchTargetLODSetting" :bApplyNaniteSettings true :NaniteSettings "(bEnabled=True,FallbackPercentTriangles=100.000000,FallbackRelativeError=0.000000)" :NewNaniteSettings (Utilities|Struct|MakeMeshNaniteSettings true) :bEmitTransaction true :bDeferMeshPostEditChange false) (Utilities|Struct|MakeGeometryScriptMeshWriteLOD false)))))

(fn ConstructionScript
  (bind _targetmesh (Variables|1SwitchMaterialIDs|GetTargetMesh))
  (bind _returnvalue (AddComponent|NODE_AddStaticMeshComponent-0 :RelativeTransform (Math|Transform|MakeTransform "500.000000,0.000000,0.000000")))
  (Components|StaticMesh|SetStaticMesh _returnvalue _targetmesh)
  (bind _returnvalue_2 (AddComponent|NODE_AddStaticMeshComponent-1 :RelativeTransform (Math|Transform|MakeTransform "1000.000000,0.000000,0.000000")))
  (Components|StaticMesh|SetStaticMesh _returnvalue_2 _targetmesh)
  (Variables|/Script/Engine.StaticMeshComponent|SetbDisallowNanite true _returnvalue_2))
